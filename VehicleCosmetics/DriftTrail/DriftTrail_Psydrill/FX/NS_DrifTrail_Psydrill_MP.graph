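# Niagara system for the Psydrill drift trail vehicle cosmetic, attaching effects to skeletal-mesh wheel sockets, apparently the rear pair.
# Two wheel-wrap emitters hold meshes and sprites on the wheels, while two others spawn short-lived speed-driven sprites and looping mesh bursts, alongside two quality-scaled ribbons.
# Spawn groups merge several layers into single emitters, and tire-FX masks gate each wheel.
# Painted-variant overrides set colors, with masks adjusting values per game mode and front-end preview.
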NiagaraSystem /VehicleCosmetics/DriftTrail/DriftTrail_Psydrill/FX/NS_DrifTrail_Psydrill_MP
dependencies {
 usedDataInterfaces [
  &NiagaraDataInterfaceCurve
  &NiagaraDataInterfaceArrayFloat3
  &NiagaraDataInterfaceArrayColor
  &NiagaraDataInterfaceVectorCurve
  &NiagaraDataInterfaceSkeletalMesh
 ]
 usedDynamicInputs []
 usedModules [
  &self:SystemSpawnScript.SystemScriptSource.SystemScriptGraph.NiagaraNodeAssignment_0.SetVariables_#1
  &self:SystemSpawnScript.SystemScriptSource.SystemScriptGraph.NiagaraNodeAssignment_1.SetVariables_#2
  &FN_UnpackVehicleFlags
  /VehicleCosmetics/FX/Niagara/Modules/VC_SetVehicleSpeedAttribs
  &SystemState
  /Game/Effects/Niagara/Modules/Vehicle/FN_UpdateVehicleAttribs
  /Game/Effects/Niagara/Modules/Vehicle/FN_VehicleFlags_Cheatcodes
 ]
}
emitters {
 NE_VC_DriftTrail_Sprite_Parent {
  emitterUpdateScript [
   &EmitterState(Inactive Response=::NewEnumerator1, Life Cycle Mode=::NewEnumerator1, Loop Duration=0.1, UseLoopDelay=false)
   &SpawnRate(
    SpawnRate=/VehicleCosmetics/FX/Niagara/Modules/VC_Mask_Floats(
     FrontEnd Value:NiagaraFloat=200
     blsFrontEndPreview:NiagaraBool=false
     SpeedMin Value:NiagaraFloat=120
     SpeedMax Value:NiagaraFloat=180
     Speed Alpha:NiagaraFloat=$System.Speed_Alpha
    )
    Use Spawn Probability=false
   )
   &SpawnBurst_Instantaneous(Spawn Count=5, Spawn Group=1, Use Spawn Probability=false, UseLoopCountLimit=false) {enabled=false}
  ]
  particleSpawnScript [
   &VC_PaintedVariantsOverride(
    Black=<rgba 0 0 0 2>
    BurntSienna=<rgba 0.5 0.1058 0 2>
    Cobalt=<rgba 0 0 3.5 2>
    Crimson=<rgba 2.5 0 0.0678 1>
    DefaultColor=&SelectLinearColorFromArray(
     Array Sampling Mode=::NewEnumerator0
     Direct Array Index=?
     Color Selection Array=DataInterface() {propertyValues {ColorData [<rgba 10 1 10 0.8>, <rgba 10 0.5 1 0.8>]}}
     Random Seed=0
    )
    ForestGreen=<rgba 0.3333 2 0 2>
    Grey=<rgba 1.5 1.5 1.5 1>
    Lime=<rgba 1.8333 2 0 1>
    Orange=<rgba 2 0.2581 0.198 1>
    Pink=<rgba 4 0.5625 2.75 1>
    Purple=<rgba 1.3333 0 2 2>
    Saffron=<rgba 2 1.5333 0 2>
    SkyBlue=<rgba 0 2.2917 2.5 1>
    TitaniumWhite=<rgba 2 2 2 1>
   )
   /SetVariables(Particles.Visivily_tag:NiagaraInt32=&SpawnGroup()) {isSetParameters=true}
   &InitializeParticle(
    Color Mode=::NewEnumerator0
    Lifetime Max=0.25
    Lifetime Min=0.25
    Lifetime Mode=::NewEnumerator1
    Sprite Rotation Mode=::NewEnumerator2
    Sprite Size Mode=::NewEnumerator1
    Uniform Sprite Size Max=4
    Uniform Sprite Size Min=3
    UsePositionOffset=false
    Write Material Random=true
   )
   &PartitionParticles()
   &SkeletalMeshLocation(
    Bone / Socket Index=&Add_Integer(A=$Particles.PartitionParticles.Partition, B=4)
    Bone / Socket Sampling Mode=::NewEnumerator5
    Interpolate Mesh Sampling=false
    MaskSpawnGroup=false
    Mesh Sampling Type=::NewEnumerator6
    Sampled Position Offset=&AddVectorToPosition(
     Position=&AddVectorToPosition(Position=<-50 0 0>, Vector=$User.offset)
     Vector=&VC_Mask_Vectors(Coro Value:Vector3f=<-50 0 -20>, FrontEnd Value:Vector3f=<0 0 0>, FN Value:Vector3f=<0 0 0>)
    )
    Skeletal Mesh=$System.SkeletalMesh
   )
   &AddVelocity(Velocity=/VehicleCosmetics/FX/Niagara/Modules/MergeEmitters/VC_VectorBySpawnGroup(
     SpawnGroup or ExecIndex:NiagaraInt32=0
     CustomInt:NiagaraInt32=0
     MaximumSpawnGroups:NiagaraInt32=2
     Vector Group 0:Vector3f=<-1000 0 0>
     Vector Group 1:Vector3f=<0 0 0>
     Vector Group 2:Vector3f=<0 0 0>
     Vector Group 3:Vector3f=<0 0 0>
     Vector Group 4:Vector3f=<0 0 0>
     Vector Group 5:Vector3f=<0 0 0>
    ))
   &ShapeLocation(MaskSpawnGroup=false, Override Seed=false, Sphere Radius=25, UseOwnerScale=true)
  ]
  particleUpdateScript [
   &FN_DisableTireFX(
    Mask Index Attribute:NiagaraInt32=$Particles.PartitionParticles.Partition
    Mask Index FL=2
    Mask Index FR=3
    Mask Index RL=0
    Mask Index RR=1
   )
   &ParticleState()
   &SolveForcesAndVelocity(Clamp Velocity=false, Limit Acceleration=false)
   &ScaleColor(Scale RGB=&VectorFromFloat(Value=1), ScaleA=true, ScaleRGB=true)
  ]
  rendererClasses [&NiagaraSpriteRendererProperties]
  renderers [{rendererClass=&NiagaraSpriteRendererProperties}]
 }
 NE_VC_DriftTrail_Sprite_Parent002 {
  emitterUpdateScript [
   &EmitterState(Inactive Response=::NewEnumerator1, Life Cycle Mode=::NewEnumerator1, Loop Duration=0.2, UseLoopDelay=false)
   &SpawnBurst_Instantaneous(Spawn Count=2, Spawn Group=1, Use Spawn Probability=false, UseLoopCountLimit=false) {module=SpawnBurst_Instantaneous001}
   &SpawnBurst_Instantaneous(Spawn Count=2, Use Spawn Probability=false, UseLoopCountLimit=false) {module=SpawnBurst_Instantaneous002}
   &SpawnRate(SpawnRate=50, Use Spawn Probability=false) {enabled=false}
  ]
  particleSpawnScript [
   &VC_PaintedVariantsOverride(
    Black=<rgba 0.01 0.01 0.01 0.5>
    BurntSienna=<rgba 0.25 0.05 0 0.35>
    Cobalt=<rgba 0.025 0.0454 1.25 0.8>
    Crimson=<rgba 1.5 0.03 0.079 0.8>
    DefaultColor=<rgba 3.5 0.07 0.2987 1>
    ForestGreen=<rgba 0 0.6 0 0.6>
    Grey=<rgba 0.5 0.5 0.5 0.5>
    Lime=<rgba 0.505 1 0.1 0.5>
    Orange=<rgba 1.5 0.09 0.09 0.6>
    Pink=<rgba 4 0.2 2.6182 0.3>
    Purple=<rgba 0.2333 0 2 1.7>
    R1 Black:LinearColor=<rgba 0.01 0.01 0.01 0.5>
    R1 BurntSienna:LinearColor=<rgba 0.25 0.05 0 0.3>
    R1 Cobalt:LinearColor=<rgba 0.0225 0.0225 1.5 0.7>
    R1 Crimson:LinearColor=<rgba 1.25 0.025 0.0862 1>
    R1 DefaultColor=&SelectLinearColorFromArray(
     Array Sampling Mode=::NewEnumerator1
     Direct Array Index=&RandomRangeInt(Minimum=0, Maximum=1)
     Color Selection Array=DataInterface() {propertyValues {ColorData [<rgba 1 0.02 0.1225 3>, <rgba 1 0.0474 0.1143 3>]}}
     Random Seed=?
    )
    R1 ForestGreen:LinearColor=<rgba 0 0.6 0 0.6>
    R1 Grey:LinearColor=<rgba 0.4 0.4 0.4 0.25>
    R1 Lime:LinearColor=<rgba 0.505 1 0.1 0.5>
    R1 Orange:LinearColor=<rgba 1.5 0.09 0.09 0.4>
    R1 Pink:LinearColor=<rgba 1 0.1823 0.7027 0.3>
    R1 Purple:LinearColor=<rgba 0.2833 0 1.7 1>
    R1 Saffron:LinearColor=<rgba 1 1 0.05 0.25>
    R1 SkyBlue:LinearColor=<rgba 0 0.3333 1 1>
    R1 TitaniumWhite:LinearColor=<rgba 1 1 1 0.5>
    Saffron=<rgba 1 1 0.06 0.6>
    SkyBlue=<rgba 0 0.3333 0.8 2>
    TitaniumWhite=<rgba 1 1 1 0.5>
    Use R1 Painted Overrides=true
   )
   /SetVariables(Particles.Visivily_tag:NiagaraInt32=&SpawnGroup()) {isSetParameters=true}
   &InitializeParticle(
    Color Mode=::NewEnumerator0
    Lifetime=/VehicleCosmetics/FX/Niagara/Modules/MergeEmitters/VC_FloatBySpawnGroup(
     SpawnGroup or ExecIndex:NiagaraInt32=0
     CustomInt:NiagaraInt32=0
     MaximumSpawnGroups:NiagaraInt32=2
     Float SG 0:NiagaraFloat=0.2
     Float SG 1:NiagaraFloat=0.2
     Float SG 2:NiagaraFloat=0
     Float SG 3:NiagaraFloat=0
     Float SG 4:NiagaraFloat=0
     Float SG 5:NiagaraFloat=0
    )
    Mesh Scale=&VC_VectorBySpawnGroup(
     SpawnGroup or ExecIndex:NiagaraInt32=0
     CustomInt:NiagaraInt32=0
     MaximumSpawnGroups:NiagaraInt32=2
     Vector Group 0:Vector3f=<0.5 0.07 0.07>
     Vector Group 1:Vector3f=<0.55 0.125 0.125>
     Vector Group 2:Vector3f=<0 0 0>
     Vector Group 3:Vector3f=<0 0 0>
     Vector Group 4:Vector3f=<0 0 0>
     Vector Group 5:Vector3f=<0 0 0>
    )
    Mesh Scale Mode=::NewEnumerator4
    Sprite Rotation Mode=::NewEnumerator2
    UsePositionOffset=false
    Write Material Random=true
   )
   &PartitionParticles()
   &SkeletalMeshLocation(
    Bone / Socket Index=&Add_Integer(A=$Particles.PartitionParticles.Partition, B=4)
    Bone / Socket Sampling Mode=::NewEnumerator5
    Interpolate Mesh Sampling=false
    MaskSpawnGroup=false
    Mesh Sampling Type=::NewEnumerator6
    Sampled Position Offset=&AddVectorToPosition(
     Position=&AddVectorToPosition(
      Position=&VC_VectorBySpawnGroup(
       SpawnGroup or ExecIndex:NiagaraInt32=0
       CustomInt:NiagaraInt32=0
       MaximumSpawnGroups:NiagaraInt32=2
       Vector Group 0:Vector3f=<-5 0 0>
       Vector Group 1:Vector3f=&SelectVectorFromArray(
        Array Sampling Mode=::NewEnumerator0
        Direct Array Index=?
        Vector Selection Array=DataInterface() {propertyValues {FloatData [<0 0 0>, <-5 0 0>]}}
        Random Seed=0
       )
       Vector Group 2:Vector3f=<0 0 0>
       Vector Group 3:Vector3f=<0 0 0>
       Vector Group 4:Vector3f=<0 0 0>
       Vector Group 5:Vector3f=<0 0 0>
      )
      Vector=$User.offset
     )
     Vector=&VC_Mask_Vectors(Coro Value:Vector3f=<-50 0 -20>, FrontEnd Value:Vector3f=<0 0 0>, FN Value:Vector3f=<0 0 0>)
    )
    Skeletal Mesh=$System.SkeletalMesh
   )
   &AddVelocity(Velocity=&VC_VectorBySpawnGroup(
     SpawnGroup or ExecIndex:NiagaraInt32=0
     CustomInt:NiagaraInt32=0
     MaximumSpawnGroups:NiagaraInt32=2
     Vector Group 0:Vector3f=<-2000 0 0>
     Vector Group 1:Vector3f=<0 0 0>
     Vector Group 2:Vector3f=<0 0 0>
     Vector Group 3:Vector3f=<0 0 0>
     Vector Group 4:Vector3f=<0 0 0>
     Vector Group 5:Vector3f=<0 0 0>
    ))
   &InitialMeshOrientation(
    Mesh Orientation Mode=::NewEnumerator2
    Orientation Vector=<-1 0 0>
    Rotation=&RandomRangeVector(Minimum=<0 0 0>, Maximum=<1 0 0>)
    Use Rotation Vector=true
   )
  ]
  particleUpdateScript [
   &FN_DisableTireFX(
    Mask Index Attribute:NiagaraInt32=$Particles.PartitionParticles.Partition
    Mask Index FL=2
    Mask Index FR=3
    Mask Index RL=0
    Mask Index RR=1
   )
   &ParticleState()
   &ScaleColor(ScaleA=true, ScaleRGB=true)
   &ScaleMeshSize(Scale Factor=&VC_VectorBySpawnGroup(
     SpawnGroup or ExecIndex:NiagaraInt32=0
     CustomInt:NiagaraInt32=0
     MaximumSpawnGroups:NiagaraInt32=2
     Vector Group 0:Vector3f=&VectorFromCurve(VectorCurve=DataInterface() {propertyValues {
        XCurve {keys [<at 0 1.5>,<at 1 0.3>]}
        YCurve {keys [<at 0 2 RCIM_Cubic>,<at 1 0.25 RCIM_Cubic>]}
        ZCurve {keys [<at 0 2 RCIM_Cubic>,<at 1 0.25 RCIM_Cubic>]}
       }})
     Vector Group 1:Vector3f=&VectorFromCurve(
      VectorCurve=DataInterface() {propertyValues {XCurve {keys [<at 0 0.5>,<at 1 4.5>]}, YCurve {keys [<at 0 0.9>,<at 1 2.5>]}, ZCurve {keys [<at 0 0.9>,<at 1 2.5>]}}}
      Scale Curve=&VectorFromFloat(Value=0.75)
     )
     Vector Group 2:Vector3f=<0 0 0>
     Vector Group 3:Vector3f=<0 0 0>
     Vector Group 4:Vector3f=<0 0 0>
     Vector Group 5:Vector3f=<0 0 0>
    ))
   &MeshRotationForce(Rotation=&SelectVectorFromArray(
     Array Sampling Mode=::NewEnumerator0
     Direct Array Index=?
     Vector Selection Array=DataInterface() {propertyValues {FloatData [<250 0 0>, <-250 0 0>]}}
     Random Seed=0
    )) {enabled=false}
   &InheritSourceMovement(
    Applied Position Scale=&Multiply_VectorByFloat(Vector=<1 1 1>, Float=1)
    ApplyForce=false
    ApplyPosition=true
    ApplyPositionFalloff=false
   ) {enabled=false}
   &SolveForcesAndVelocity(Clamp Velocity=false, Limit Acceleration=false)
   /SetVariables(Particles.ScaleSizeR1:Vector3f=&VectorFromFloat(Value=&FloatFromCurve(FloatCurve=DataInterface() {propertyValues {Curve {
         defaultValue=0.8
         keys [
          {arriveTangent=9.0725, interpMode=RCIM_Cubic, leaveTangent=9.0725, tangentMode=RCTM_User, time=0, value=0}
          {arriveTangent=-2.5249, interpMode=RCIM_Cubic, leaveTangent=-2.5249, tangentMode=RCTM_User, time=0.1408, value=1}
          {arriveTangent=-0.0405, interpMode=RCIM_Cubic, leaveTangent=-0.0405, tangentMode=RCTM_User, time=1, value=0}
         ]
        }}}))) {isSetParameters=true}
   &DynamicMaterialParameters(
    Index 0 Param 1=&FloatFromCurve(FloatCurve=DataInterface() {propertyValues {Curve {keys [{tangentMode=RCTM_User, time=0, value=1}, {tangentMode=RCTM_User, time=1, value=0}]}}})
    Index 0 Param 2=&FloatFromCurve(
     FloatCurve=DataInterface() {propertyValues {Curve {defaultValue=0.5227, keys [{tangentMode=RCTM_User, time=0, value=0.5}, {tangentMode=RCTM_User, time=1, value=0}]}}}
     Scale Curve=&VC_FloatBySpawnGroup(
      SpawnGroup or ExecIndex:NiagaraInt32=0
      CustomInt:NiagaraInt32=0
      MaximumSpawnGroups:NiagaraInt32=2
      Float SG 0:NiagaraFloat=9
      Float SG 1:NiagaraFloat=20
      Float SG 2:NiagaraFloat=0
      Float SG 3:NiagaraFloat=0
      Float SG 4:NiagaraFloat=0
      Float SG 5:NiagaraFloat=0
     )
    )
    Index 1 Param 1=&FloatFromCurve(FloatCurve=DataInterface() {propertyValues {Curve {keys [<at 0 1>,<at 1 0>]}}})
    Index 1 Param 2=1
    Module.Index0Float0Write=true
    Module.Index0Float1Write=true
    Module.Index0Float2Write=true
    Module.Index0Float3Write=true
    Module.Index1Float0Write=true
    Module.Index1Float1Write=true
    Module.Index1Float2Write=true
    Module.Index1Float3Write=true
    Write Parameter Index 1=::NewEnumerator1
   ) {enabled=false}
  ]
  rendererClasses [&NiagaraMeshRendererProperties, &NiagaraSpriteRendererProperties]
  renderers [
   {rendererClass=&NiagaraMeshRendererProperties, rendererIndex=0}
   {rendererClass=&NiagaraMeshRendererProperties, rendererIndex=1}
   {rendererClass=&NiagaraSpriteRendererProperties, rendererIndex=2}
  ]
 }
 NE_VC_DriftTrail_WheelWrap_Parent {
  emitterUpdateScript [
   &EmitterState(
    Inactive Response=::NewEnumerator1
    Life Cycle Mode=::NewEnumerator1
    Loop Behavior=::NewEnumerator1
    Loop Duration Mode=::NewEnumerator1
    UseLoopDelay=false
   )
   &SpawnBurst_Instantaneous(Spawn Count=6, Use Spawn Probability=false, UseLoopCountLimit=false) {module=SpawnBurst_Instantaneous001}
   &SpawnBurst_Instantaneous(Spawn Count=6, Spawn Group=1, Use Spawn Probability=false, UseLoopCountLimit=false) {module=SpawnBurst_Instantaneous002}
   &SpawnBurst_Instantaneous(Spawn Count=6, Spawn Group=2, Use Spawn Probability=false, UseLoopCountLimit=false) {module=SpawnBurst_Instantaneous003}
   &SpawnBurst_Instantaneous(Spawn Count=2, Spawn Group=3, Use Spawn Probability=false, UseLoopCountLimit=false) {module=SpawnBurst_Instantaneous004}
   &SpawnBurst_Instantaneous(Spawn Count=6, Spawn Group=2, Use Spawn Probability=false, UseLoopCountLimit=false) {enabled=false}
  ]
  particleSpawnScript [
   &ApplyInitialForces()
   &VC_PaintedVariantsOverride(
    Black=<rgba 0.001 0.001 0.001 1>
    BurntSienna=<rgba 0.22 0.0389 0.011 1>
    Cobalt=<rgba 0.035 0.0927 3.5 1>
    Crimson=<rgba 1 0.0167 0 1>
    DefaultColor=<rgba 7 0.035 0.2091 1>
    ForestGreen=<rgba 0.075 0.9 0 1>
    Grey=<rgba 0.5 0.5 0.5 1>
    Lime=<rgba 0.505 1 0.1 1>
    Orange=<rgba 3 0.1372 0.03 1>
    Pink=<rgba 2 0.4271 1.428 1>
    Purple=<rgba 0.5 0.3 1.5 1>
    R1 Black:LinearColor=<rgba 0.0191 0.0191 0.0191 3>
    R1 BurntSienna:LinearColor=<rgba 1 0.3424 0.1233 0.5>
    R1 Cobalt:LinearColor=<rgba 0 0.0555 3 1>
    R1 Crimson:LinearColor=<rgba 2 0.4 0.4434 0.5>
    R1 DefaultColor=<rgba 1 0.8922 0.931 1>
    R1 ForestGreen:LinearColor=<rgba 0 0.5 0.0253 1>
    R1 Gold:LinearColor=<rgba 5 5 5 1>
    R1 Grey:LinearColor=<rgba 0.35 0.35 0.35 1>
    R1 Lime:LinearColor=<rgba 0.55 1 0.1 1.15>
    R1 Orange:LinearColor=<rgba 2 0.5 0 1>
    R1 Pink:LinearColor=<rgba 2 0.2396 1.3598 1>
    R1 Purple:LinearColor=<rgba 0.1667 0 2 5>
    R1 Saffron:LinearColor=<rgba 1.3 1.3 0 1>
    R1 SkyBlue:LinearColor=<rgba 0.24 0.8788 1.2 1>
    R1 TitaniumWhite:LinearColor=<rgba 2 2 2 1>
    R2 Black:LinearColor=<rgba 0.25 0.25 0.25 3>
    R2 BurntSienna:LinearColor=<rgba 1.2 0.155 0.06 5>
    R2 Cobalt:LinearColor=<rgba 0 0.1 3 10>
    R2 Crimson:LinearColor=<rgba 4 0.08 0.2107 10>
    R2 DefaultColor:LinearColor=<rgba 3 3 3 15>
    R2 ForestGreen:LinearColor=<rgba 0.4 2 0.4808 5>
    R2 Gold:LinearColor=<rgba 20 20 20 1>
    R2 Grey:LinearColor=<rgba 5 5 5 1>
    R2 Lime:LinearColor=<rgba 7.5 15 0 1>
    R2 Orange:LinearColor=<rgba 3 0.216 0.12 50>
    R2 Pink:LinearColor=<rgba 7 2.3333 5.303 1>
    R2 Purple:LinearColor=<rgba 10.3386 1.0156 15 1>
    R2 Saffron:LinearColor=<rgba 10 8.0167 1.5 1>
    R2 SkyBlue:LinearColor=<rgba 2.25 6.5 15 1>
    R2 TitaniumWhite:LinearColor=<rgba 20 20 20 1>
    R3 Black:LinearColor=<rgba 0.1 0.1 0.1 0.7>
    R3 BurntSienna:LinearColor=<rgba 1 0.205 0.1 1>
    R3 Cobalt:LinearColor=<rgba 0 0.0925 5 0.5>
    R3 Crimson:LinearColor=<rgba 6 0.06 0.221 1>
    R3 DefaultColor:LinearColor=<rgba 3 0.3 0.435 0.5>
    R3 ForestGreen:LinearColor=<rgba 0.3 2 0.3859 1>
    R3 Grey:LinearColor=<rgba 1.5 1.5 1.5 1>
    R3 Lime:LinearColor=<rgba 0.8625 1.5 0.225 1>
    R3 Orange:LinearColor=<rgba 2 0.5 0.2 1>
    R3 Pink:LinearColor=<rgba 3 0.3594 2.0398 0.5>
    R3 Purple:LinearColor=<rgba 0.5 0 2 1>
    R3 Saffron:LinearColor=<rgba 2 2 0.1 1>
    R3 SkyBlue:LinearColor=<rgba 0 1.8333 2 1>
    R3 TitaniumWhite:LinearColor=<rgba 2 2 2 1>
    Saffron=<rgba 0.8 0.8 0.024 1>
    SkyBlue=<rgba 0 0.525 0.9 1>
    TitaniumWhite=<rgba 2 2 2 1>
    Use R1 Painted Overrides=true
    Use R2 Painted Overrides=true
    Use R3 Painted Overrides=true
   ) {module=VC_PaintedVariantsOverride001}
   /SetVariables(Particles.Visivily_tag:NiagaraInt32=&SpawnGroup()) {isSetParameters=true}
   &InitializeParticle(
    Color Mode=::NewEnumerator0
    Lifetime=&VC_FloatBySpawnGroup(
     SpawnGroup or ExecIndex:NiagaraInt32=0
     CustomInt:NiagaraInt32=0
     MaximumSpawnGroups:NiagaraInt32=4
     Float SG 0:NiagaraFloat=0
     Float SG 1:NiagaraFloat=0
     Float SG 2:NiagaraFloat=0
     Float SG 3:NiagaraFloat=0
     Float SG 4:NiagaraFloat=0
     Float SG 5:NiagaraFloat=0
    )
    Sprite Size Mode=::NewEnumerator3
    Uniform Sprite Size=&MakeCustomFloatFromBool(Bool=$User.bIsPainted, True Float=35, False Float=50)
    UsePositionOffset=false
    Write Material Random=true
   )
   /SetVariables(
    Particles.SCALE Rate 1:Vector3f=<0.3 0.165 0.165>
    Particles.SCALE Rate 2:Vector3f=<0.2 0.22 0.22>
    Particles.SCALE Rate 3:Vector3f=<0.3 0.165 0.165>
    Particles.SCALE Rate 4:Vector3f=<0 0 0>
   ) {isSetParameters=true}
   &PartitionParticles()
   &SkeletalMeshLocation(
    Bone / Socket Index=&Add_Integer(A=$Particles.PartitionParticles.Partition, B=4)
    Bone / Socket Sampling Mode=::NewEnumerator5
    Interpolate Mesh Sampling=false
    MaskSpawnGroup=false
    Mesh Sampling Type=::NewEnumerator6
    Sampled Position Offset=&VC_VectorBySpawnGroup(
     SpawnGroup or ExecIndex:NiagaraInt32=0
     CustomInt:NiagaraInt32=0
     MaximumSpawnGroups:NiagaraInt32=3
     Vector Group 0:Vector3f=<-110 0 -15>
     Vector Group 1:Vector3f=<0 0 -15>
     Vector Group 2:Vector3f=<135 0 -15>
     Vector Group 3:Vector3f=<0 0 0>
     Vector Group 4:Vector3f=<0 0 0>
     Vector Group 5:Vector3f=<0 0 0>
    )
    Skeletal Mesh=$System.SkeletalMesh
   )
   &InitialMeshOrientation(Mesh Orientation Mode=::NewEnumerator2, Orientation Vector=<-1 0 0>, Rotation=<130000 0 0>, Use Rotation Vector=true)
   &AlignSpriteToMeshOrientation(Mesh Orientation Relative Sprite Alignment Vector=<1 0 0>, SetAlignment=true, SetFacing=false) {enabled=false}
  ]
  particleUpdateScript [
   &FN_DisableTireFX(
    Mask Index Attribute:NiagaraInt32=$Particles.PartitionParticles.Partition
    Mask Index FL=2
    Mask Index FR=3
    Mask Index RL=0
    Mask Index RR=1
   )
   &VC_WheelWrapScale(Wheel Index Attribute:NiagaraInt32=&Add_Integer(A=$Particles.PartitionParticles.Partition, B=2))
   &ParticleState(Kill Particles When Lifetime Has Elapsed=false, Loop Particles Lifetime=true)
   &ScaleSpriteSize(Uniform Curve Sprite Scale=DataInterface() {propertyValues {Curve {defaultValue=1, keys [{tangentMode=RCTM_User, time=0, value=0.5}, {tangentMode=RCTM_User, time=1, value=1}]}}})
   &UpdateMeshOrientation(Rotation Coordinate Space=::NewEnumerator2, Rotation Vector=<0 1 0>) {enabled=false}
   &SkeletalMeshLocation(
    Bone / Socket Index=&Add_Integer(A=$Particles.PartitionParticles.Partition, B=4)
    Bone / Socket Sampling Mode=::NewEnumerator5
    MaskSpawnGroup=false
    Mesh Sampling Type=::NewEnumerator6
    Sampled Position Offset=&AddVectorToPosition(
     Position=&AddVectorToPosition(
      Position=&VC_VectorBySpawnGroup(
       SpawnGroup or ExecIndex:NiagaraInt32=0
       CustomInt:NiagaraInt32=0
       MaximumSpawnGroups:NiagaraInt32=4
       Vector Group 0:Vector3f=<-35 0 0>
       Vector Group 1:Vector3f=<-23 0 0>
       Vector Group 2:Vector3f=<-35 0 0>
       Vector Group 3:Vector3f=/Niagara/DynamicInputs/TypeConversions/MakeCustomPositionFromBool(Bool:NiagaraBool=$User.bIsPainted, True Position:NiagaraPosition=<-40 0 0>, False Position:NiagaraPosition=<-48 0 0>)
       Vector Group 4:Vector3f=<0 0 0>
       Vector Group 5:Vector3f=<0 0 0>
      )
      Vector=$User.offset
     )
     Vector=&VC_Mask_Vectors(Coro Value:Vector3f=<-50 0 -20>, FrontEnd Value:Vector3f=<0 0 0>, FN Value:Vector3f=<0 0 0>)
    )
    Skeletal Mesh=$System.SkeletalMesh
   ) {module=SkeletalMeshLocation001}
   &DynamicMaterialParameters(
    Index 0 Param 1=&VC_FloatBySpawnGroup(
     SpawnGroup or ExecIndex:NiagaraInt32=0
     CustomInt:NiagaraInt32=0
     MaximumSpawnGroups:NiagaraInt32=3
     Float SG 0:NiagaraFloat=&MakeCustomFloatFromBool(Bool=$User.bIsPainted, True Float=0.2, False Float=0.8)
     Float SG 1:NiagaraFloat=0.05
     Float SG 2:NiagaraFloat=0.03
     Float SG 3:NiagaraFloat=0
     Float SG 4:NiagaraFloat=0
     Float SG 5:NiagaraFloat=0
    )
    Index 3 Param 1=&MakeCustomFloatFromBool(Bool=$User.bIsPainted, True Float=0, False Float=1)
    Module.Index0Float0Write=true
    Module.Index0Float1Write=true
    Module.Index0Float2Write=true
    Module.Index0Float3Write=true
    Module.Index3Float0Write=true
    Module.Index3Float1Write=true
    Module.Index3Float2Write=true
    Module.Index3Float3Write=true
    Write Parameter Index 3=::NewEnumerator1
   )
   &SolveForcesAndVelocity(Clamp Velocity=false, Limit Acceleration=false)
   /SetVariables(
    Particles.ORB_COLOR:LinearColor=&VC_DynamicScaleColor(
     Color to Scale:LinearColor=$Particles.R3 Color
     RGB Scale:Vector3f=&VectorFromFloat(Value=&FloatFromCurve(
       FloatCurve=DataInterface() {propertyValues {Curve {
          defaultValue=0
          keys [{interpMode=RCIM_Cubic, tangentMode=RCTM_User, time=0, value=1}, {interpMode=RCIM_Cubic, tangentMode=RCTM_User, time=1, value=1}]
         }}}
       Scale Curve=2
      ))
     Alpha Scale:NiagaraFloat=1
    )
    Particles.SCALE COLOR VEIL:LinearColor=&VC_DynamicScaleColor(
     Color to Scale:LinearColor=$Particles.R1 Color
     RGB Scale:Vector3f=&VectorFromFloat(Value=1)
     Alpha Scale:NiagaraFloat=1
    )
   ) {isSetParameters=true}
  ]
  rendererClasses [&NiagaraMeshRendererProperties, &NiagaraSpriteRendererProperties]
  renderers [
   {rendererClass=&NiagaraMeshRendererProperties, rendererIndex=0}
   {rendererClass=&NiagaraMeshRendererProperties, rendererIndex=1}
   {rendererClass=&NiagaraSpriteRendererProperties, rendererIndex=2}
   {rendererClass=&NiagaraMeshRendererProperties, rendererIndex=3}
  ]
 }
 NE_VC_DriftTrail_WheelWrap_Parent002 {
  emitterUpdateScript [
   &EmitterState(
    Inactive Response=::NewEnumerator1
    Life Cycle Mode=::NewEnumerator1
    Loop Behavior=::NewEnumerator1
    Loop Duration Mode=::NewEnumerator1
    UseLoopDelay=false
   )
   &SpawnBurst_Instantaneous(Spawn Count=2, Use Spawn Probability=false, UseLoopCountLimit=false)
  ]
  particleSpawnScript [
   &VC_PaintedVariantsOverride(
    Black=<rgba 0.15 0.15 0.15 1>
    BurntSienna=<rgba 0.75 0.3037 0.225 1>
    Cobalt=<rgba 0.6 0.8567 5 2.5>
    Crimson=<rgba 2 0.5 0.5 0.7>
    DefaultColor=<rgba 2 0.258 0.6239 0.1217>
    ForestGreen=<rgba 0.645 1.8 0.54 2>
    Grey=<rgba 0.5 0.5 0.5 3>
    Lime=<rgba 1.0125 1.5 0.525 0.8>
    Orange=<rgba 2 0.708 0.48 1>
    Pink=<rgba 3 0.75 2.1818 0.5>
    Purple=<rgba 0.8775 0.455 1.3 4>
    R1 DefaultColor=<rgba 10 0.7812 1.0885 0.6>
    Saffron=<rgba 1.9767 2 0.6 0.5>
    SkyBlue=<rgba 0.52 1.39 4 3>
    TitaniumWhite=<rgba 1 1 1 1.5>
    Use R1 Painted Overrides=true
   )
   &VC_PaintedVariantsOverride(
    Black=<rgba 0.05 0.05 0.05 55>
    BurntSienna=<rgba 0.2 0.0618 0.0247 1>
    Cobalt=<rgba 0 0.0925 5 1>
    Crimson=<rgba 10 0 0.271 1>
    DefaultColor=<rgba 20 0 4.3511 1>
    ForestGreen=<rgba 0 5 0.2526 1>
    Lime=<rgba 2.5 5 0 1>
    Orange=<rgba 3 1.0464 0 1>
    Purple=<rgba 1.2312 0 3 1>
    R1 BurntSienna:LinearColor=<rgba 0.1 0.0309 0.0123 1>
    R1 Cobalt:LinearColor=<rgba 0 0.037 2 1>
    R1 Crimson:LinearColor=<rgba 1 0 0.0271 1>
    R1 DefaultColor=<rgba 10 0 2.1755 1>
    R1 ForestGreen:LinearColor=<rgba 0 0.6 0.0303 1>
    R1 Lime:LinearColor=<rgba 0.25 0.5 0 1>
    R1 Orange:LinearColor=<rgba 1 0.3488 0 1>
    R1 Purple:LinearColor=<rgba 0.4104 0 1 1>
    R1 Saffron:LinearColor=<rgba 0.8 0.6133 0 1>
    R1 SkyBlue:LinearColor=<rgba 0 0.5323 0.8 1>
    R1 TitaniumWhite:LinearColor=<rgba 0.5 0.5 0.5 1>
    Saffron=<rgba 2 1.5333 0 1>
    SkyBlue=<rgba 0 3.3269 5 1>
    TitaniumWhite=<rgba 2 2 2 1>
    Use R1 Painted Overrides=true
   ) {enabled=false, module=VC_PaintedVariantsOverride001}
   &InitializeParticle(
    Color=$Particles.Color
    Mesh Scale Mode=::NewEnumerator4
    Sprite Rotation Mode=::NewEnumerator2
    Sprite Size Mode=::NewEnumerator3
    UsePositionOffset=false
    Write Material Random=true
   )
   &PartitionParticles()
   &SkeletalMeshLocation(
    Bone / Socket Index=&Add_Integer(A=$Particles.PartitionParticles.Partition, B=4)
    Bone / Socket Sampling Mode=::NewEnumerator5
    Interpolate Mesh Sampling=false
    MaskSpawnGroup=false
    Mesh Sampling Type=::NewEnumerator6
    Skeletal Mesh=$System.SkeletalMesh
   )
  ]
  particleUpdateScript [
   &FN_DisableTireFX(
    Mask Index Attribute:NiagaraInt32=$Particles.PartitionParticles.Partition
    Mask Index FL=2
    Mask Index FR=3
    Mask Index RL=0
    Mask Index RR=1
   )
   &VC_WheelWrapScale(Wheel Index Attribute:NiagaraInt32=&Add_Integer(A=$Particles.PartitionParticles.Partition, B=2))
   &ParticleState(Kill Particles When Lifetime Has Elapsed=false)
   &ScaleMeshSize(Scale Factor=&SelectVectorFromArray(
     Direct Array Index=$Particles.PartitionParticles.Partition
     Vector Selection Array=DataInterface() {propertyValues {FloatData [<-0.95 -0.95 0.95>, <0.95 0.95 0.95>]}}
    )) {enabled=false, module=ScaleMeshSize001}
   &UpdateMeshOrientation(Rotation Coordinate Space=::NewEnumerator2, Rotation Vector=<0 1 0>) {enabled=false}
   &ScaleMeshSize(Scale Factor=&SelectVectorFromArray(
     Direct Array Index=$Particles.PartitionParticles.Partition
     Vector Selection Array=DataInterface() {propertyValues {FloatData [<1 -1 -1>, <1 1 1>]}}
    )) {enabled=false}
   &SkeletalMeshLocation(
    Bone / Socket Index=&Add_Integer(A=$Particles.PartitionParticles.Partition, B=4)
    Bone / Socket Sampling Mode=::NewEnumerator5
    MaskSpawnGroup=false
    Mesh Sampling Type=::NewEnumerator6
    Sampled Position Offset=&SelectVectorFromArray(
     Direct Array Index=$Particles.PartitionParticles.Partition
     Vector Selection Array=DataInterface() {propertyValues {FloatData [<0 0 0>, <0 1 0>]}}
    )
    Skeletal Mesh=$System.SkeletalMesh
   ) {module=SkeletalMeshLocation001}
   &SolveForcesAndVelocity(Clamp Velocity=false, Limit Acceleration=false)
  ]
  rendererClasses [&NiagaraMeshRendererProperties]
  renderers [
   {rendererClass=&NiagaraMeshRendererProperties, rendererIndex=0}
   {rendererClass=&NiagaraMeshRendererProperties, rendererIndex=1}
   {rendererClass=&NiagaraMeshRendererProperties, rendererIndex=2}
  ]
 }
 Ribbons001 {
  emitterUpdateScript [
   &EmitterState(Inactive Response=::NewEnumerator1, Life Cycle Mode=::NewEnumerator1, Loop Duration=0.5, UseLoopDelay=false)
   &VC_RateQualityHandler(High Spawn Rate=60, Low Spawn Rate=30, Medium Spawn Rate=45, Use Custom Rate?=true)
   &VC_SpawnRibbon_v1(Number of Ribbons=2, SpawnRate=$Emitter.Spawn Rate)
  ]
  particleSpawnScript [
   &VC_PaintedVariantsOverride(
    Black=<rgba 0.005 0.005 0.005 1>
    BurntSienna=/VehicleCosmetics/FX/Niagara/Modules/VC_Mask_LinearColor(
     Coro Color:LinearColor=<rgba 0.25 0.056 0.0075 0.08>
     Coro FrontEnd Color:LinearColor=<rgba 0.1 0.0224 0.003 0.08>
     FN Color:LinearColor=<rgba 0.25 0.056 0.0075 0.08>
    )
    Cobalt=&VC_Mask_LinearColor(
     Coro Color:LinearColor=<rgba 0.0083 0 0.5 0.8>
     Coro FrontEnd Color:LinearColor=<rgba 0.0029 0 0.175 0.8>
     FN Color:LinearColor=<rgba 0.0083 0 0.5 0.8>
    )
    Crimson=&VC_Mask_LinearColor(
     Coro Color:LinearColor=<rgba 2 0.04 0.0727 0.5>
     Coro FrontEnd Color:LinearColor=<rgba 0.15 0.003 0.0055 0.5>
     FN Color:LinearColor=<rgba 2 0.04 0.0727 0.5>
    )
    DefaultColor=&VC_Mask_LinearColor(
     Coro Color:LinearColor=<rgba 1 0.09 0.1203 0.4>
     Coro FrontEnd Color:LinearColor=<rgba 0.125 0.0112 0.015 0.4>
     FN Color:LinearColor=<rgba 1 0.09 0.1203 0.4>
    )
    ForestGreen=&VC_Mask_LinearColor(
     Coro Color:LinearColor=<rgba 0 1 0.0333 0.2>
     Coro FrontEnd Color:LinearColor=<rgba 0 0.075 0.0025 0.2>
     FN Color:LinearColor=<rgba 0 1 0.0333 0.2>
    )
    Grey=&VC_Mask_LinearColor(
     Coro Color:LinearColor=<rgba 0.2 0.2 0.2 0.2>
     Coro FrontEnd Color:LinearColor=<rgba 0.05 0.05 0.05 0.2>
     FN Color:LinearColor=<rgba 0.2 0.2 0.2 0.2>
    )
    Lime=&VC_Mask_LinearColor(
     Coro Color:LinearColor=<rgba 0.45 1 0 0.5>
     Coro FrontEnd Color:LinearColor=<rgba 0.0337 0.075 0 0.5>
     FN Color:LinearColor=<rgba 0.45 1 0 0.5>
    )
    Orange=&VC_Mask_LinearColor(
     Coro Color:LinearColor=<rgba 5 0.2633 0.1 0.4>
     Coro FrontEnd Color:LinearColor=<rgba 0.1 0.0053 0.002 0.4>
     FN Color:LinearColor=<rgba 5 0.2633 0.1 0.4>
    )
    Pink=&VC_Mask_LinearColor(
     Coro Color:LinearColor=<rgba 5 0.599 3.3996 0.05>
     Coro FrontEnd Color:LinearColor=<rgba 0.1 0.012 0.068 0.05>
     FN Color:LinearColor=<rgba 5 0.599 3.3996 0.05>
    )
    Purple=&VC_Mask_LinearColor(
     Coro Color:LinearColor=<rgba 0.9167 0 1 0.08>
     Coro FrontEnd Color:LinearColor=<rgba 0.0917 0 0.1 0.08>
     FN Color:LinearColor=<rgba 0.9167 0 1 0.08>
    )
    R1 Black:LinearColor=<rgba 0.01 0.01 0.01 4>
    R1 BurntSienna:LinearColor=<rgba 0.1 0.0309 0.0123 2.5>
    R1 Cobalt:LinearColor=<rgba 0 0.0462 2.5 2>
    R1 Crimson:LinearColor=<rgba 1.5 0.195 0.15 2>
    R1 DefaultColor=<rgba 2 0.4688 0.5964 0.3478>
    R1 ForestGreen:LinearColor=<rgba 0 1 0.0505 2>
    R1 Grey:LinearColor=<rgba 0.4 0.4 0.4 2>
    R1 Lime:LinearColor=<rgba 0.3667 1 0 1.5>
    R1 Orange:LinearColor=<rgba 3 1 0 0.8>
    R1 Pink:LinearColor=<rgba 5 0.2604 3.2765 1>
    R1 Purple:LinearColor=<rgba 0.5 0 2.5 2.5>
    R1 Saffron:LinearColor=<rgba 1.5 1.5 0 0.8>
    R1 SkyBlue:LinearColor=<rgba 0 1 3 1>
    R1 TitaniumWhite:LinearColor=<rgba 1.2399 1.25 1 1>
    Saffron=&VC_Mask_LinearColor(
     Coro Color:LinearColor=<rgba 4.1667 5 0 0.1>
     Coro FrontEnd Color:LinearColor=<rgba 0.0625 0.075 0 0.1>
     FN Color:LinearColor=<rgba 4.1667 5 0 0.1>
    )
    SkyBlue=&VC_Mask_LinearColor(Coro Color:LinearColor=<rgba 0 0 2.5 0.4>, Coro FrontEnd Color:LinearColor=<rgba 0 0 0.15 0.4>, FN Color:LinearColor=<rgba 0 0 2.5 0.4>)
    TitaniumWhite=&VC_Mask_LinearColor(
     Coro Color:LinearColor=<rgba 0.8 0.8 0.8 0.4>
     Coro FrontEnd Color:LinearColor=<rgba 0.05 0.05 0.05 0.4>
     FN Color:LinearColor=<rgba 0.8 0.8 0.8 0.4>
    )
    Use R1 Painted Overrides=true
   )
   &VC_InitializeRibbon_v1(
    Color=$Particles.Color
    Custom Offset=true
    Enable Skeletal Position=true
    Lifetime=&MakeCustomFloatFromBool(Bool=$User.bIsFrontendPreview, True Float=0.35, False Float=0.42)
    Local Vector Offset=&AddVectorToPosition(
     Position=&VC_Mask_Vectors(Coro Value:Vector3f=<15 0 -20>, FrontEnd Value:Vector3f=<50 0 0>, FN Value:Vector3f=<50 0 0>)
     Vector=$User.offset
    )
    Offset Amount=1
    Ribbon Width=100
   )
   &AddVelocity(Velocity=<3000 0 0>, Velocity Speed Scale=-1) {module=AddVelocity001}
  ]
  particleUpdateScript [
   &FN_DisableTireFX(Mask Index Attribute:NiagaraInt32=$Particles.SpawnGroup, Mask Index FL=2, Mask Index FR=3, Mask Index RL=0, Mask Index RR=1)
   &InheritSourceMovement(
    Applied Position Scale=&VectorFromFloat(Value=&FloatFromCurve(FloatCurve=DataInterface() {propertyValues {Curve {keys [<at 0.3 1 RCIM_Cubic>, {tangentMode=RCTM_User, time=1, value=0}]}}}))
    ApplyForce=false
    ApplyPosition=true
    ApplyPositionFalloff=false
   )
   &ParticleState()
   &ScaleColor(
    Scale Alpha=&MakeCustomFloatFromBool(
     Bool=$User.bIsPainted
     True Float=&FloatFromCurve(
      FloatCurve=DataInterface() {propertyValues {Curve {
         defaultValue=1
         keys [
          {interpMode=RCIM_Cubic, leaveTangent=0.0205, tangentMode=RCTM_Break, time=0, value=0}
          {interpMode=RCIM_Cubic, tangentMode=RCTM_User, time=0.5, value=1}
          {tangentMode=RCTM_User, time=1, value=0}
         ]
        }}}
      Scale Curve=1.25
     )
     False Float=&FloatFromCurve(FloatCurve=DataInterface() {propertyValues {Curve {
         defaultValue=1
         keys [
          {interpMode=RCIM_Cubic, leaveTangent=0.0205, tangentMode=RCTM_Break, time=0.05, value=0}
          {interpMode=RCIM_Cubic, tangentMode=RCTM_User, time=0.5, value=1}
          {tangentMode=RCTM_User, time=1, value=0}
         ]
        }}})
    )
    Scale RGB=&VectorFromFloat(Value=&FloatFromCurve(
      FloatCurve=DataInterface() {propertyValues {Curve {keys [<at 0 0 RCIM_Cubic>, {arriveTangent=1, interpMode=RCIM_Cubic, leaveTangent=1, time=0.4, value=8.5}, <at 1 1>]}}}
      Scale Curve=&MakeCustomFloatFromBool(Bool=$User.bIsPainted, True Float=10, False Float=10)
     ))
    ScaleA=true
    ScaleRGB=true
   )
   &ScaleRibbonWidth(Ribbon Width Scale=&FloatFromCurve(FloatCurve=DataInterface() {propertyValues {Curve {
        defaultValue=1
        keys [
         <at 0 0.5>
         {arriveTangent=-2.7878, interpMode=RCIM_Cubic, leaveTangent=-2.7878, tangentMode=RCTM_User, time=0.2, value=1}
         <at 1 0.1 RCIM_Cubic>
        ]
       }}}))
   &SolveForcesAndVelocity(Clamp Velocity=false, Limit Acceleration=false)
   &VC_UpdateRibbon_v1(Enable Skeletal Position=true)
   /SetVariables(
    Particles.Index0:Vector4f=&MakeVector4(X=1, Y=0, Z=0, W=0)
    Particles.MainTrailColor:LinearColor=&VC_DynamicScaleColor(
     Color to Scale:LinearColor=$Particles.R1 Color
     RGB Scale:Vector3f=&VectorFromFloat(Value=&FloatFromCurve(FloatCurve=DataInterface() {propertyValues {Curve {keys [{tangentMode=RCTM_User, time=0, value=0.5}, {tangentMode=RCTM_User, time=0.5, value=1}, <at 1 0>]}}}))
     Alpha Scale:NiagaraFloat=&FloatFromCurve(FloatCurve=DataInterface() {propertyValues {Curve {keys [<at 0 0 RCIM_Cubic>,<at 0.3 1 RCIM_Cubic>]}}}, Scale Curve=0.5)
    )
    Particles.MainTrailWidth:NiagaraFloat=&FloatFromCurve(
     FloatCurve=DataInterface() {propertyValues {Curve {keys [
         {interpMode=RCIM_Cubic, tangentMode=RCTM_User, time=0, value=0.5}
         {arriveTangent=0.0747, interpMode=RCIM_Cubic, leaveTangent=0.0747, tangentMode=RCTM_User, time=0.05, value=1}
         {arriveTangent=0.0113, interpMode=RCIM_Cubic, leaveTangent=0.0113, tangentMode=RCTM_User, time=0.1, value=0.45}
         {arriveTangent=0.2778, interpMode=RCIM_Cubic, leaveTangent=0.2778, tangentMode=RCTM_User, time=0.15, value=0.7}
         {arriveTangent=-0.1456, interpMode=RCIM_Cubic, leaveTangent=-0.1456, tangentMode=RCTM_User, time=0.9, value=0}
        ]}}}
     Scale Curve=75
    )
   ) {isSetParameters=true}
  ]
  rendererClasses [&NiagaraRibbonRendererProperties]
  renderers [
   {rendererClass=&NiagaraRibbonRendererProperties}
   {rendererClass=&NiagaraRibbonRendererProperties, rendererIndex=1}
  ]
 }
}
summary {
 systemName=NS_DrifTrail_Psydrill_MP
 emitters [
  {
   emitterName=NE_VC_DriftTrail_WheelWrap_Parent
   rendererClasses [&NiagaraMeshRendererProperties, &NiagaraSpriteRendererProperties]
  }
  {emitterName=Ribbons001, rendererClasses [&NiagaraRibbonRendererProperties]}
  {emitterName=NE_VC_DriftTrail_WheelWrap_Parent002, rendererClasses [&NiagaraMeshRendererProperties]}
  {emitterName=NE_VC_DriftTrail_Sprite_Parent, rendererClasses [&NiagaraSpriteRendererProperties]}
  {
   emitterName=NE_VC_DriftTrail_Sprite_Parent002
   rendererClasses [&NiagaraMeshRendererProperties, &NiagaraSpriteRendererProperties]
  }
 ]
 userVariables [
  User.bIsFrontendPreview {type=NiagaraBool}
  User.bIsPainted {type=NiagaraBool}
  User.bPreviewVehicleSpeed {type=NiagaraBool}
  User.ColorVariantID {type=NiagaraInt32}
  User.GameModeID {type=NiagaraInt32, defaultValue=3}
  User.offset {type=Vector3f, description="offset from wheels(set through cosmetic template)", defaultValue {struct=Vector3f, value=<0 0 0>}}
  User.PaintedColor {type=LinearColor, description="", defaultValue {struct=LinearColor, value=<rgba 1 1 1 1>}}
  User.PaintedScalar {type=NiagaraFloat, defaultValue=1, description=""}
  User.VehicleFlags {type=NiagaraInt32, defaultValue=806289408}
  User.Wheel_Radius_Array {
   type=NiagaraDataInterfaceArrayFloat
   defaultValue {
    struct=NiagaraExt_VariableValue_DataInterface
    value {
     dataInterface=&self:NiagaraDataInterfaceArrayFloat_1
     dataInterfaceClass=&NiagaraDataInterfaceArrayFloat
    }
   }
  }
  User.Wheel_Width_Array {
   type=NiagaraDataInterfaceArrayFloat
   defaultValue {
    struct=NiagaraExt_VariableValue_DataInterface
    value {
     dataInterface=&self:NiagaraDataInterfaceArrayFloat_0
     dataInterfaceClass=&NiagaraDataInterfaceArrayFloat
    }
   }
  }
  User.WheelTouchingArray {
   type=NiagaraDataInterfaceArrayFloat
   defaultValue {
    struct=NiagaraExt_VariableValue_DataInterface
    value {
     dataInterface=&self:NiagaraDataInterfaceArrayFloat_2
     dataInterfaceClass=&NiagaraDataInterfaceArrayFloat
    }
   }
  }
 ]
}
